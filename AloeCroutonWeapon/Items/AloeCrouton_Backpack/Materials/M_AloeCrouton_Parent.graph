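# Parent character material for AloeCrouton items, layering clear coat and effects between the 2023 PreFX and PostFX functions.
# With IsBackpack on, it renders a masked screen with distorted static, scanlines, sweep line and a logo fill driven by current versus max charges.
# With it off, parameters blend idle, pulsing overheat and curve-atlas cooldown emissive, using barrel and exhaust masks decoded from the Tech Art Mask.
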Material /AloeCroutonWeapon/Items/AloeCrouton_Backpack/Materials/M_AloeCrouton_Parent
graph {
 properties {
  bUseMaterialAttributes=true
  bUsedWithMeshDeformer=true
  bUsedWithNanite=true
  bUsedWithSkeletalMesh=true
  pixelDepthOffsetMode=PDOM_Legacy
  shadingModel=MSM_FromMaterialExpression
  usedShadingModels=Default Lit | Subsurface | Clear Coat | Subsurface Profile | Cloth
 }
 nodes [
  n1 = MaterialFunctionCall() {materialFunction=&MF_CharacterParent_2023_PreFX}
  n2 = MaterialFunctionCall!(Mask=@n1.Tech Art Mask, Fresnel=@n1.Tech Art Mask, In=@n1.Tech Art Mask) {materialFunction=/Game/Global/MaterialFunctions/TechArt/MF_Advanced_CarClearCoat}
  ClearCoatNormalCustomOutput!(None=MaterialFunctionCall(BaseNormal=@n2.NormalWithOrangePeel, AdditionalNormal=@n2.NormalWithOrangePeel) {materialFunction=&BlendAngleCorrectedNormals}.Result)
  Composite!(None=@n1.Tech Art Mask) {subgraphName=ProcessMasks}
  MaterialFunctionCall(Material=NamedRerouteUsage(), M=@n1.M, Pre Skinned Local Position=@n1.M, Distance Blend=@n1.M, Fresnel=@n1.M) {materialFunction=&MF_CharacterParent_2023_PostFX}
  NamedRerouteDeclaration(None=@n2.Result) {name=Material In, nodeColor=<rgba 0 1 0.2266 1>}
  n3 = GetMaterialAttributes(None=NamedRerouteUsage()) {attributeGetTypes [
    #1
    #2
    #3
    #4
   ]}
  n4 = ScalarParameter() {group=AloeCrouton, parameterName=AC_ScreenDiffuseMultiplier, sortPriority=100}
  n5 = ScalarParameter() {group=AloeCrouton, parameterName=AC_ScreenSpecularMultiplier, sortPriority=100}
  n6 = SetMaterialAttributes!(
   MaterialAttributes=@n3.MaterialAttributes
   Base Color=LinearInterpolate(A=@n3.Base Color, B=Multiply(A=@n3.Base Color, B=@n4) {constB=0.01}, Alpha=NamedRerouteUsage())
   Specular=LinearInterpolate(A=@n3.Specular, B=Multiply(A=@n3.Specular, B=@n5), Alpha=NamedRerouteUsage()) {constB=0.1}
   Emissive Color=Add(A=@n3.Emissive Color, B=NamedRerouteUsage())
   Opacity=LinearInterpolate(A=@n3.Opacity, Alpha=NamedRerouteUsage()) {constB=0, desc="Might not be needed\r\nAfter getting final textures"}
  ) {attributeSetTypes [
    #1
    #2
    #3
    #4
   ]}
  n7 = GetMaterialAttributes(None=NamedRerouteUsage()) {attributeGetTypes [#3]}
  n8 = NamedRerouteUsage()
  n9 = ScalarParameter() {defaultValue=2, group=Emissive, parameterName=AC_IdleMainIntensity}
  n10 = NamedRerouteUsage()
  n11 = Multiply(
   A=VectorParameter() {group=Emissive, parameterName=AC_OverheatColor, defaultValue=<rgba 10 2.5114 0 1>}.RGB
   B=ScalarParameter() {defaultValue=1, group=Emissive, parameterName=AC_OverheatIntensity}
  )
  n12 = Time()
  n13 = ConstantBiasScale(None=Sine(None=Multiply(A=@n12) {constB=4})) {bias=1.2}
  n14 = ScalarParameter() {group=Emissive, parameterName=AC_OverheatBlend}
  n15 = Multiply(
   A=@n11
   B=Multiply(
    A=ScalarParameter() {defaultValue=0.5, group=Emissive, parameterName=AC_OverheatExhaustChargeFactor}
    B=LinearInterpolate(B=@n13, Alpha=@n14) {constA=1, constB=0.5, desc=Pulse exhaust during charge}
   )
  )
  n16 = NamedRerouteUsage()
  n17 = NamedRerouteUsage()
  n18 = Multiply(
   A=LinearInterpolate(A=Multiply(A=@n10, B=@n11), B=Multiply(A=@n15, B=@n16), Alpha=@n17) {constB=0.5}
   B=ScalarParameter() {group=Emissive, parameterName=AC_OverheatBlend}
  )
  n19 = Multiply(
   A=Multiply(A=@n17, B=NamedRerouteUsage())
   B=Multiply(A=@n11, B=ScalarParameter() {defaultValue=0.05, group=Emissive, parameterName=AC_OverheatExhaustIdleFactor})
  )
  n20 = NamedRerouteUsage()
  n21 = Constant() {r=1}
  n22 = Multiply(
   A=Add(A=Desaturation(None=@n20, Fraction=@n21), B=NamedRerouteUsage())
   B=ScalarParameter() {defaultValue=0.25, group=Emissive, parameterName=AC_OverheatMainCooldownIntensity}
  ) {constB=0.5}
  n23 = Multiply(A=ScalarParameter() {defaultValue=0.25, group=Emissive, parameterName=AC_OverheatExhaustCooldownIntensity}, B=NamedRerouteUsage())
  n24 = NamedRerouteUsage()
  n25 = NamedRerouteUsage()
  n26 = CurveAtlasRowParameter(CurveTime=Multiply(A=LinearInterpolate(A=@n22, B=@n23, Alpha=@n24) {constB=0.5}, B=OneMinus(None=@n25))) {atlas=&AthenaCurveAtlas01, curve=/Game/Athena/LinearColorCurves/IceKingFlameHeart, defaultValue=36}
  n27 = Add(
   A=Add(A=Multiply(A=@n8, B=@n9), B=Add(A=@n18, B=@n19)) {desc=plasma visuals}
   B=Multiply(A=@n26, B=ScalarParameter() {group=Emissive, parameterName=AC_CooldownIntensity}) {desc=heat visuals}
  )
  NamedRerouteDeclaration(None=StaticSwitchParameter(
    True=@n6
    False=SetMaterialAttributes!(MaterialAttributes=@n7.MaterialAttributes, Emissive Color=@n27) {attributeSetTypes [#3]}
   ) {defaultValue=true, parameterName=IsBackpack}) {name=Material Out, nodeColor=<rgba 0 1 0.6953 1>}
  n28 = NamedRerouteUsage()
  n29 = ComponentMask(None=Frac(None=@n28)) {g=true}
  n30 = ScalarParameter() {defaultValue=2, group=AloeCrouton, parameterName=AC_Charges_Current, sortPriority=10}
  n31 = Divide(A=@n30, B=ScalarParameter() {defaultValue=2, group=AloeCrouton, parameterName=AC_Charges_Max, sortPriority=10})
  n32 = Add(A=@n29, B=@n31) {constB=0.1}
  n33 = Frac(None=Multiply(A=@n28, B=ScalarParameter() {defaultValue=2, group=AloeCrouton, parameterName=AC_Charges_Max, sortPriority=10}))
  n34 = Power(Base=OneMinus(None=ComponentMask(None=@n33) {g=true})) {constExponent=3}
  NamedRerouteDeclaration(None=Multiply(A=Floor(None=@n32), B=Add(A=@n34) {constB=0.5})) {name=ChargeCounter, nodeColor=<rgba 1 0.7031 0 1>}
  NamedRerouteDeclaration(None=ComponentMask(None=@n1.Tech Art Mask) {r=true}) {name=OverheatEmissive, nodeColor=<rgba 0 0.1328 1 1>}
  MaterialExpressionReroute_4 = Reroute(None=@n1.Tech Art Mask)
  n35 = ComponentMask(None=@MaterialExpressionReroute_4.Input) {b=true}
  n36 = Floor(None=Add(A=@n35) {constB=0.15})
  MaterialExpressionReroute_5 = Reroute(None=@n36)
  NamedRerouteDeclaration(None=@MaterialExpressionReroute_5.BarrelMask) {name=BarrelMask, nodeColor=<rgba 0.7344 0 1 1>}
  n37 = Subtract(A=Floor(None=Add(A=@n35) {constB=0.3}), B=@n36)
  MaterialExpressionReroute_6 = Reroute(None=@n37)
  NamedRerouteDeclaration(None=@MaterialExpressionReroute_6.ExhaustMask) {name=ExhaustMask, nodeColor=<rgba 1 0.3516 0 1>}
  NamedRerouteDeclaration(None=@n7.Emissive Color) {name=BaseEmissive, nodeColor=<rgba 0.2188 0 1 1>}
  n38 = VertexInterpolator(VS=ComponentMask(None=PreSkinnedNormal()) {b=true, g=true, r=true})
  n39 = Abs(None=ComponentMask(None=@n38.PS) {g=true})
  n40 = Constant() {r=0.5}
  n41 = Saturate(None=ComponentMask(None=@n38.PS) {b=true})
  n42 = Constant() {r=2}
  n43 = Saturate(None=Subtract(
    A=MaterialFunctionCall(In=@n39, Contrast=@n40) {materialFunction=&CheapContrast}.Result
    B=MaterialFunctionCall(In=@n41, Contrast=@n42) {materialFunction=&CheapContrast}.Result
   ))
  n44 = Constant()
  n45 = MaterialFunctionCall(
   UVs=Add(A=BumpOffset(Height=@n44) {heightRatio=0.1}, B=Constant2Vector() {g=0.29, r=-0.358})
   Texture Scale=Constant2Vector() {g=0.2, r=0.125}
  ) {materialFunction=&ScaleUVsByCenter}
  NamedRerouteDeclaration(None=Multiply(
    A=OneMinus(None=@n43)
    B=TextureSample(UVs=@n45.UVs) {samplerSource=SSM_Clamp_WorldGroupSettings, texture=/Game/Athena/HUD/Materials/T_UI_RadialGradient}.G
   )) {name=ExhaustEmissive, nodeColor=<rgba 1 0.0994 0 1>}
  n46 = ComponentMask(None=TextureCoordinate()) {r=true}
  n47 = Multiply(A=Time()) {constB=-0.03}
  n48 = ConstantBiasScale(None=Sine(None=Add(A=@n46, B=@n47)) {period=0.02}) {bias=1.1}
  n49 = Divide(A=OneMinus(None=@n29), B=@n31)
  n50 = Power(Base=Saturate(None=@n49), Exp=Multiply(A=Constant() {r=32}, B=@n30)) {constExponent=16}
  n51 = OneMinus(None=Floor(None=@n49))
  NamedRerouteDeclaration(None=Multiply(A=Multiply(A=@n48) {constB=4}, B=Multiply(A=@n50, B=@n51))) {name=TopLine, nodeColor=<rgba 0.4531 0 1 1>}
  n52 = TextureSampleParameter2D() {
   group=AloeCrouton
   parameterName=AC_ScreenMask
   samplerSource=SSM_Clamp_WorldGroupSettings
   sortPriority=30
   texture=/AloeCroutonWeapon/Items/AloeCrouton_Backpack/Textures/T_AloeCrouton_Backbling_Mask
  }
  NamedRerouteDeclaration(None=@n52.R) {name=Screen Mask, nodeColor=<rgba 0 1 0.4844 1>}
  NamedRerouteDeclaration(None=Saturate(None=Subtract(A=@n52.G, B=@n52.G))) {name=Glow Mask, nodeColor=<rgba 1 0.1172 0 1>}
  n53 = TextureCoordinate()
  n54 = VectorParameter() {parameterName=LogoCenter, defaultValue=<rgba 0.313 0.7875 0 0>}
  n55 = MaterialFunctionCall(
   UVs=Subtract(A=Add(A=@n53) {constB=0.5}, B=ComponentMask(None=@n54.RGB) {g=true, r=true})
   Texture Scale=ScalarParameter() {defaultValue=0.08, group=AloeCrouton, parameterName=LogoScale, sortPriority=30}
  ) {materialFunction=&ScaleUVsByCenter}
  n56 = TextureSampleParameter2D(UVs=Add(A=NamedRerouteDeclaration(None=@n55.UVs) {name=CenteredUVs, nodeColor=<rgba 0 1 0.7422 1>}, B=NamedRerouteUsage())) {
   group=AloeCrouton
   parameterName=AC_LogoMask
   samplerSource=SSM_Clamp_WorldGroupSettings
   samplerType=SAMPLERTYPE_Masks
   sortPriority=30
   texture=/AloeCroutonWeapon/Items/AloeCrouton_Backpack/Textures/T_AloeCrouton_Backbling_LogoMask
  }
  NamedRerouteDeclaration(None=Multiply(A=NamedRerouteUsage(), B=@n56.R)) {name=LogoMask, nodeColor=<rgba 0 1 0.0625 1>}
  n57 = VectorParameter() {group=Screen Static, parameterName=TexCoordScale, defaultValue=<rgba 5 9 0 0>}
  NamedRerouteDeclaration(None=Multiply(A=TextureCoordinate(), B=ComponentMask(None=@n57.RGB) {g=true, r=true})) {name=TexCoord, nodeColor=<rgba 1 0 0.4453 1>}
  n58 = NamedRerouteUsage()
  n59 = Multiply(
   A=TextureSample(UVs=Panner(Coordinate=@n58) {speedX=6, speedY=-0.05}) {samplerType=SAMPLERTYPE_Masks, texture=/Game/Missions/Primary/RepairTheShelter/Textures/Cinematic_Screen_Static}.G
   B=TextureSample(UVs=Panner(Coordinate=@n58) {speedX=-3, speedY=-0.03}) {samplerType=SAMPLERTYPE_Masks, texture=&Cinematic_Screen_Static}.R
  )
  n60 = Multiply(A=@n59, B=ScalarParameter() {defaultValue=0.07, group=Static, parameterName=Static UV Distort})
  n61 = Constant()
  n62 = Multiply(A=ChannelMaskParameter(None=NamedRerouteUsage()) {maskChannel=Green, parameterName=Glitch Direction, defaultValue=<rgba 0 1 0 0>}.R) {constB=0.75}
  n63 = Multiply(A=Time()) {constB=-0.2}
  n64 = Power(Base=Frac(None=Add(A=@n62, B=@n63))) {constExponent=20}
  n65 = Multiply(A=@n64, B=ScalarParameter() {defaultValue=-0.015, group=Sweep Line, parameterName=Sweep UV Distort}) {constB=0.02}
  n66 = ComponentMask(None=NamedRerouteUsage()) {r=true}
  n67 = MaterialFunctionCall(X=Sign(None=Subtract(A=@n66) {constB=0.5})) {materialFunction=&MakeFloat2}
  NamedRerouteDeclaration(None=Add(A=AppendVector(A=@n60, B=@n61), B=Multiply(A=@n65, B=@n67.Result))) {name=Distorted UVs, nodeColor=<rgba 0 0.3203 1 1>}
  n68 = NamedRerouteUsage()
  n69 = ConstantBiasScale(None=Frac(None=Panner(Coordinate=@n68) {speedY=-1})) {bias=-0.5, scale=2}
  n70 = Multiply(A=ComponentMask(None=Power(Base=@n69)) {g=true}, B=ScalarParameter() {defaultValue=0.5, group=Background, parameterName=Scanline Strength}) {constB=0.1}
  n71 = Multiply(A=TextureSample(UVs=Panner(Coordinate=@n58) {speedX=5, speedY=0.1}) {samplerType=SAMPLERTYPE_Masks, texture=&Cinematic_Screen_Static}.B) {constB=0.1}
  n72 = Add(
   A=Multiply(A=Add(A=@n71, B=@n59), B=ScalarParameter() {defaultValue=6, group=Static, parameterName=Static Scale})
   B=Multiply(A=@n64, B=ScalarParameter() {defaultValue=0.2, group=Sweep Line, parameterName=Sweep Add}) {constB=0.2}
  )
  n73 = ScalarParameter() {defaultValue=0.2, group=Background, parameterName=Background}
  NamedRerouteDeclaration(None=Multiply(A=Add(A=@n70), B=Add(A=@n72, B=@n73)) {constA=1}) {name=Static, nodeColor=<rgba 0.4531 0 1 1>}
  n74 = VectorParameter() {group=AloeCrouton, parameterName=AC_ScreenColor, sortPriority=50, defaultValue=<rgba 0.0292 0.15 0.0072 1>}
  n75 = VectorParameter() {group=AloeCrouton, parameterName=LogoGlowOff, sortPriority=50, defaultValue=<rgba 0.013 0.063 0.004 1>}
  n76 = VectorParameter() {group=AloeCrouton, parameterName=LogoGlowOn, sortPriority=50, defaultValue=<rgba 0.2502 0.7157 0.0545 1>}
  n77 = Multiply(A=NamedRerouteUsage(), B=ScalarParameter() {defaultValue=8, parameterName=TopLineBrightness})
  n78 = Multiply(A=Multiply(A=@n76.RGB, B=Add(A=@n77)), B=ScalarParameter() {defaultValue=1, parameterName=AC_Charges_EmissiveIntensity})
  n79 = NamedRerouteUsage()
  n80 = Add(
   A=Multiply(A=@n74.RGB, B=NamedRerouteUsage())
   B=Multiply(A=NamedRerouteUsage(), B=LinearInterpolate(A=@n75.RGB, B=@n78, Alpha=@n79) {constAlpha=1})
  )
  n81 = NamedRerouteUsage()
  n82 = VectorParameter() {group=AloeCrouton, parameterName=AC_ScreenFalloffGlow, sortPriority=50, defaultValue=<rgba 0.1738 0.5 0.0408 1>}
  n83 = NamedRerouteUsage()
  NamedRerouteDeclaration(None=Add(A=Multiply(A=@n80, B=@n81), B=Multiply(A=@n82.RGB, B=@n83))) {name=Screen Emissive, nodeColor=<rgba 1 0.7969 0 1>}
  PinBase() {reroutePins [Input {expression=&self:MaterialExpressionReroute_4}]}
  PinBase!(BarrelMask=@n36, ExhaustMask=@n37) {reroutePins [
    BarrelMask {expression=&self:MaterialExpressionReroute_5}
    ExhaustMask {expression=&self:MaterialExpressionReroute_6}
   ]}
 ]
}
